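# Niagara lightning effect for the SatireCane_Ode emote, showing a glowing green-teal logo mesh amid crackling bolts.
# A brief intro mesh burst leads into the roughly three-second main logo, which has curve-driven material parameters, flare sprite and light.
# Short-lived bolts spawn at a tapering rate with randomized orientation, scale and arc sprites.
# A delayed impact burst adds meshes, sprites and a light flash.
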NiagaraSystem /BRCosmetics/Animation/Game/MainPlayer/Emotes/SatireCane_Ode/FX/NS_SatireCaneOde_LightningBolt
dependencies {
 usedDataInterfaces [
  &NiagaraDataInterfaceCurve
  &NiagaraDataInterfaceArrayColor
  &NiagaraDataInterfaceColorCurve
  &NiagaraDataInterfaceVectorCurve
 ]
 usedDynamicInputs [&SimulationPosition]
 usedModules [
  &SystemState
  &self:SystemSpawnScript.SystemScriptSource.SystemScriptGraph.NiagaraNodeAssignment_0.SetVariables_#1
 ]
}
emitters {
 Intro_Logo_Main001 {
  emitterUpdateScript [
   &EmitterState(Life Cycle Mode=::NewEnumerator1, Loop Behavior=::NewEnumerator1, Loop Delay=0.2, Loop Duration=0.1, UseLoopDelay=false)
   &SpawnBurst_Instantaneous(
    Spawn Count=1
    Spawn Probability=&FloatFromCurve(FloatCurve=DataInterface() {propertyValues {Curve {keys [<at 0.5 1 RCIM_Cubic>,<at 1 0>]}}}, CurveIndex=$System.NormalizedLoopAge)
    Use Spawn Probability=false
   )
  ]
  particleSpawnScript [
   &InitializeParticle(
    Color=<rgba 0 0.5 0.2332 1>
    Lifetime=0.35
    Mesh Scale=<1.7 1.3 1>
    Mesh Scale Mode=::NewEnumerator4
    Position Offset=<-100 0 280>
    Sprite Rotation Mode=::NewEnumerator2
    UsePositionOffset=false
    Write Material Random=true
   ) {module=InitializeParticle001}
   &InitialMeshOrientation(Mesh Orientation Mode=::NewEnumerator7, Rotation=<0 0 0.5>, Use Rotation Vector=true)
  ]
  particleUpdateScript [
   &ParticleState()
   &ScaleMeshSize(Scale Factor=<1.25 1.25 1.25>) {module=ScaleMeshSize001}
   &UpdateMeshOrientation(
    Facing Direction=<0 1 0>
    Facing Interpolation=true
    Orientation Method=::NewEnumerator1
    Override Quat=false
    TransformFacing=false
    TransformUp=false
   )
   &ScaleMeshSize(Scale Factor=&VectorFromCurve(VectorCurve=DataInterface() {propertyValues {
       XCurve {keys [<at 0 1 RCIM_Cubic>,<at 0.5 1>,<at 0.8 1>,<at 1 1.2>]}
       YCurve {keys [<at 0 1 RCIM_Cubic>,<at 0.5 1>,<at 0.6 1>,<at 1 1.2>]}
       ZCurve {keys [<at 0 1 RCIM_Cubic>,<at 0.5 1>,<at 0.8 1>,<at 1 1.1>]}
      }}))
   &ScaleColor(
    Linear Color Curve=DataInterface() {propertyValues {
      AlphaCurve {keys [<at 0.5 1 RCIM_Cubic>,<at 1 0>]}
      BlueCurve {keys [<at 0 10>,<at 1 0>]}
      GreenCurve {keys [<at 0 10>,<at 1 0>]}
      RedCurve {keys [<at 0 10>,<at 1 0>]}
     }}
    Scale Mode=::NewEnumerator2
   )
   &DynamicMaterialParameters(
    Index 0 Param 1=&FloatFromCurve(FloatCurve=DataInterface() {propertyValues {Curve {defaultValue=1, keys [<at 0 0 RCIM_Cubic>, {tangentMode=RCTM_User, time=1, value=0.1}]}}})
    Index 0 Param 3=&FloatFromCurve(FloatCurve=DataInterface() {propertyValues {Curve {defaultValue=1, keys [<at 0 0 RCIM_Cubic>,<at 1 1 RCIM_Cubic>]}}})
    Index 0 Param 4=&FloatFromCurve(FloatCurve=DataInterface() {propertyValues {Curve {
        defaultValue=1
        keys [
         {arriveTangent=5.5593, interpMode=RCIM_Cubic, leaveTangent=5.5593, tangentMode=RCTM_User, time=0.75, value=-0.5}
         {tangentMode=RCTM_User, time=1, value=0.25}
        ]
       }}})
    Index 1 Param 1=&FloatFromCurve(FloatCurve=DataInterface() {propertyValues {Curve {keys [<at 0 0 RCIM_Cubic>,<at 0.75 1>]}}})
    Module.Index0Float0Write=true
    Module.Index0Float1Write=true
    Module.Index0Float2Write=true
    Module.Index0Float3Write=true
    Module.Index1Float0Write=true
    Module.Index1Float1Write=true
    Module.Index1Float2Write=true
    Module.Index1Float3Write=true
    Write Parameter Index 1=::NewEnumerator1
   ) {module=DynamicMaterialParameters001}
  ]
  rendererClasses [&NiagaraMeshRendererProperties]
  renderers [{rendererClass=&NiagaraMeshRendererProperties, rendererIndex=0}]
 }
 Lightning_Bolts {
  emitterUpdateScript [
   &EmitterState(
    Inactive Response=::NewEnumerator1
    Life Cycle Mode=::NewEnumerator1
    Loop Behavior=::NewEnumerator1
    Loop Delay=0.15
    Loop Duration=3.12
    UseLoopDelay=true
   )
   &SpawnRate(SpawnGroup=3, SpawnRate=10, Use Spawn Probability=false) {enabled=false, module=SpawnRate001}
   &SpawnRate(
    SpawnRate=&FloatFromCurve(FloatCurve=DataInterface() {propertyValues {Curve {keys [<at 0.7 15 RCIM_Cubic>,<at 1 0>]}}}, CurveIndex=$Emitter.NormalizedLoopAge)
    Use Spawn Probability=false
   )
   &SpawnBurst_Instantaneous(
    Spawn Count=&UniformRangedInt(Minimum=1, Maximum=1)
    Spawn Probability=&FloatFromCurve(FloatCurve=DataInterface() {propertyValues {Curve {keys [<at 0.5 1 RCIM_Cubic>,<at 1 0>]}}}, CurveIndex=$System.NormalizedLoopAge)
    Use Spawn Probability=false
   )
  ]
  particleSpawnScript [
   /SetVariables(Particles.VisibilityTag=&SpawnGroup()) {enabled=false, isSetParameters=true}
   &InitializeParticle(
    Color=&SelectLinearColorFromArray(
     Array Sampling Mode=::NewEnumerator1
     Direct Array Index=0
     Color Selection Array=DataInterface() {propertyValues {ColorData [<rgba 0 0.5 0.2332 1>, <rgba 0 1 0.4664 1>]}}
     Random Seed=?
    )
    Lifetime=&UniformRangedFloat(Minimum=0.2, Maximum=0.25)
    Mesh Scale Max=<3 4 4>
    Mesh Scale Min=<2 3 3>
    Mesh Scale Mode=::NewEnumerator2
    Sprite Rotation Mode=::NewEnumerator2
    Sprite Size Max=<200 200>
    Sprite Size Min=<150 150>
    Sprite Size Mode=::NewEnumerator2
    UsePositionOffset=true
    Write Material Random=true
   ) {module=InitializeParticle001}
   /SetVariables(
    Particles.FlareOffset:Vector3f=&UniformRangedVector(Minimum=<0 0 -0.5>, Maximum=<8 0 0.5>)
    Particles.FlareSize:Vector2f=&Vector2DFromFloat(Value=&UniformRangedFloat(Minimum=100, Maximum=150))
    Particles.LensFlare_Color:LinearColor=&MultiplyLinearColor(
     Linear Color=$Particles.Color
     Scale Factor=&MakeVector4_FromLinearColor(LinearColor=$System.RandomFlareColor)
    )
   ) {isSetParameters=true, module=SetVariables_#2}
   &InitialMeshOrientation(
    Mesh Axis To Orient=<0 0 -1>
    Mesh Orientation Mode=::NewEnumerator2
    Orientation Coordinate Space=::World
    Orientation Vector=<0 0 0>
    Rotation=&RandomRangeVector(Minimum=<0 -0.1 -0.1>, Maximum=<1 0.3 0.1>)
    Use Rotation Vector=true
   ) {module=InitialMeshOrientation001}
   &CameraOffset(Camera Offset Amount=5)
   &InitialMeshOrientation(
    Mesh Orientation Mode=::NewEnumerator1
    Rotation=&RandomRangeVector(Minimum=<0 0 0>, Maximum=<1 0 1>)
    Use Rotation Vector=true
   )
   /SetVariables(
    Particles.Arc_Position:NiagaraPosition=&AddVectorToPosition(
     Position=$Engine.Emitter.SimulationPosition
     Vector=&RandomRangeVector(Minimum=<-15 -15 0>, Maximum=<15 15 200>)
    )
    Particles.Arc_Size:Vector2f=&RandomRangeVector2D(Minimum=<20 20>, Maximum=<75 75>)
   ) {isSetParameters=true}
   /SetVariables() {enabled=false, isSetParameters=true}
   &SpriteFacingAndAlignment(Alignment Coordinate Space=::Local, Facing Coordinate Space=::Local, Sprite Alignment=<0 0 1>, WriteAlignment=true, WriteFacing=true) {enabled=false}
  ]
  particleUpdateScript [
   &ParticleState()
   &ScaleSpriteSize(Uniform Curve Sprite Scale=DataInterface() {propertyValues {Curve {keys [<at 0 0 RCIM_Cubic>, {tangentMode=RCTM_User, time=1, value=1}]}}})
   &ScaleColor(
    Scale Alpha=&FloatFromCurve(FloatCurve=DataInterface() {propertyValues {Curve {defaultValue=0.9847, keys [<at 0 1 RCIM_Cubic>, {tangentMode=RCTM_User, time=1, value=0}]}}})
    Scale RGB=&MakeVectorFromLinearColorRGB(Color=&ColorFromCurve(ColorCurve=DataInterface() {propertyValues {
        AlphaCurve {keys []}
        BlueCurve {keys [<at 0 3 RCIM_Cubic>,<at 1 0 RCIM_Cubic>]}
        GreenCurve {keys [<at 0 2.798 RCIM_Cubic>,<at 1 0 RCIM_Cubic>]}
        RedCurve {keys [<at 0 2.2969 RCIM_Cubic>,<at 1 0 RCIM_Cubic>]}
       }}))
    ScaleA=true
    ScaleRGB=true
   )
   &FN_Light_Attributes(
    Color=<rgba 0 0.35 0.1632 1>
    Create Color Attribute=true
    Create Exponent Attribute=true
    Create Radius Attribute=true
    Exponent=2
    Radius=300
    Scale Color=&FloatFromCurve(
     FloatCurve=DataInterface() {propertyValues {Curve {
        defaultValue=1
        keys [
         {arriveTangent=9.0725, interpMode=RCIM_Cubic, leaveTangent=9.0725, tangentMode=RCTM_User, time=0, value=0}
         {arriveTangent=-2.5249, interpMode=RCIM_Cubic, leaveTangent=-2.5249, tangentMode=RCTM_User, time=0.1408, value=0.9847}
         {arriveTangent=-0.0405, interpMode=RCIM_Cubic, leaveTangent=-0.0405, tangentMode=RCTM_User, time=1, value=0}
        ]
       }}}
     Scale Curve=20
    )
   ) {enabled=false}
   &DynamicMaterialParameters(
    Index 0 Param 1=10
    Index 0 Param 2=1
    Index 0 Param 3=&FloatFromCurve(FloatCurve=DataInterface() {propertyValues {Curve {defaultValue=1, keys [{tangentMode=RCTM_User, time=0, value=0}, {tangentMode=RCTM_User, time=1, value=1}]}}})
    Index 0 Param 4=0.01
    Index 1 Param 1=&FloatFromCurve(FloatCurve=DataInterface() {propertyValues {Curve {
        defaultValue=1
        keys [
         {arriveTangent=3.7522, interpMode=RCIM_Cubic, leaveTangent=3.7522, tangentMode=RCTM_User, time=0, value=0.5}
         {tangentMode=RCTM_User, time=0.25, value=1}
        ]
       }}})
    Index 1 Param 2=1
    Index 1 Param 3=1
    Index 2 Param 2=&FloatFromCurve(FloatCurve=DataInterface() {propertyValues {Curve {defaultValue=0, keys [<at 0 1 RCIM_Cubic>, {tangentMode=RCTM_User, time=1, value=0.2}]}}})
    Module.Index0Float0Write=true
    Module.Index0Float1Write=true
    Module.Index0Float2Write=true
    Module.Index0Float3Write=true
    Module.Index1Float0Write=true
    Module.Index1Float1Write=true
    Module.Index1Float2Write=true
    Module.Index1Float3Write=true
    Module.Index2Float0Write=true
    Module.Index2Float1Write=true
    Module.Index2Float2Write=false
    Module.Index2Float3Write=false
    Module.Index3Float0Write=false
    Module.Index3Float1Write=false
    Module.Index3Float2Write=false
    Module.Index3Float3Write=true
    Write Parameter Index 1=::NewEnumerator1
    Write Parameter Index 2=::NewEnumerator1
    Write Parameter Index 3=::NewEnumerator1
   ) {module=DynamicMaterialParameters002}
   /SetVariables(Particles.Arc_DynParam:Vector4f=&MakeVector4(
     X=1
     Y=0
     Z=&FloatFromCurve(FloatCurve=DataInterface() {propertyValues {Curve {keys [<at 0.2 0 RCIM_Cubic>, {interpMode=RCIM_Cubic, leaveTangent=1.5819, time=1, value=1}]}}})
     W=0
    )) {isSetParameters=true}
  ]
  rendererClasses [
   &NiagaraMeshRendererProperties
   &NiagaraSpriteRendererProperties
   &NiagaraLightRendererProperties
  ]
  renderers [
   {rendererClass=&NiagaraMeshRendererProperties, rendererIndex=0}
   {rendererClass=&NiagaraSpriteRendererProperties, rendererIndex=1}
   {rendererClass=&NiagaraLightRendererProperties, rendererIndex=2}
   {rendererClass=&NiagaraSpriteRendererProperties, rendererIndex=3}
   {rendererClass=&NiagaraSpriteRendererProperties, rendererIndex=4}
   {rendererClass=&NiagaraSpriteRendererProperties, rendererIndex=5}
  ]
 }
 Lightning_Bolts_Impact {
  emitterUpdateScript [
   &EmitterState(
    Inactive Response=::NewEnumerator1
    Life Cycle Mode=::NewEnumerator1
    Loop Behavior=::NewEnumerator1
    Loop Delay=0.05
    Loop Duration=3.5
    UseLoopDelay=true
   )
   &SpawnRate(
    SpawnRate=&FloatFromCurve(FloatCurve=DataInterface() {propertyValues {Curve {keys [<at 0.5 20 RCIM_Cubic>,<at 1 0>]}}}, CurveIndex=$Emitter.NormalizedLoopAge)
    Use Spawn Probability=false
   ) {enabled=false}
   &SpawnBurst_Instantaneous(
    Spawn Count=3
    Spawn Probability=&FloatFromCurve(FloatCurve=DataInterface() {propertyValues {Curve {keys [<at 0.5 1 RCIM_Cubic>,<at 1 0>]}}}, CurveIndex=$System.NormalizedLoopAge)
    Spawn Time=0.1
    Use Spawn Probability=false
   )
  ]
  particleSpawnScript [
   &InitializeParticle(
    Color=<rgba 0.2346 1 0.9473 10>
    Lifetime=&UniformRangedFloat(Minimum=0.125, Maximum=0.15)
    Mesh Scale Max=<7 8 8>
    Mesh Scale Min=<2 4 4>
    Mesh Scale Mode=::NewEnumerator2
    Sprite Rotation Mode=::NewEnumerator2
    Sprite Size Max=<200 100>
    Sprite Size Min=<400 200>
    Sprite Size Mode=::NewEnumerator2
    UsePositionOffset=false
    Write Material Random=true
   ) {module=InitializeParticle001}
   &DynamicMaterialParameters(
    Index 0 Param 1=100
    Index 0 Param 2=3
    Index 0 Param 3=&UniformRangedFloat(Minimum=1.45, Maximum=2.2)
    Index 0 Param 4=&UniformRangedFloat(Minimum=0, Maximum=1)
    Module.Index0Float0Write=true
    Module.Index0Float1Write=true
    Module.Index0Float2Write=true
    Module.Index0Float3Write=true
   )
   /SetVariables(
    Particles.FlareOffset:Vector3f=&UniformRangedVector(Minimum=<0 0 -0.5>, Maximum=<8 0 0.5>)
    Particles.FlareSize:Vector2f=&Vector2DFromFloat(Value=&UniformRangedFloat(Minimum=10, Maximum=15))
    Particles.LensFlare_Color:LinearColor=&MultiplyLinearColor(
     Linear Color=$Particles.Color
     Scale Factor=&MakeVector4_FromLinearColor(LinearColor=$System.RandomFlareColor)
    )
   ) {isSetParameters=true, module=SetVariables_#2}
   &InitialMeshOrientation(
    Mesh Axis To Orient=<0 0 0>
    Mesh Orientation Mode=::NewEnumerator2
    Orientation Vector=<0 0 1>
    Rotation=&RandomRangeVector(Minimum=<0 -0.3 0>, Maximum=<1 1 1>)
    Use Rotation Vector=true
   ) {module=InitialMeshOrientation001}
   &CameraOffset(Camera Offset Amount=5)
  ]
  particleUpdateScript [
   &ParticleState()
   &ScaleSpriteSize()
   &DynamicMaterialParameters(
    Index 0 Param 1=500
    Index 0 Param 2=&FloatFromCurve(FloatCurve=DataInterface() {propertyValues {Curve {keys [{arriveTangent=-2.9283, interpMode=RCIM_Cubic, leaveTangent=-2.9283, tangentMode=RCTM_User, time=0.75, value=1.5}, <at 1 0>]}}})
    Index 0 Param 3=&Lerp_Float(A=0, B=1, Alpha=0)
    Index 0 Param 4=1
    Module.Index0Float0Write=true
    Module.Index0Float1Write=true
    Module.Index0Float2Write=true
    Module.Index0Float3Write=true
   ) {module=DynamicMaterialParameters001}
   &ScaleColor(
    Linear Color Curve=DataInterface() {propertyValues {
      AlphaCurve {keys [<at 0.0012 0>,<at 0.0923 1>,<at 1 0>]}
      BlueCurve {keys [<at 0 10>,<at 1 1>]}
      GreenCurve {keys [<at 0 10>,<at 1 1>]}
      RedCurve {keys [<at 0 10>,<at 1 1>]}
     }}
    Scale Mode=::NewEnumerator2
   )
   &ScaleMeshSize(Scale Factor=&VectorFromCurve(VectorCurve=DataInterface() {propertyValues {
       XCurve {keys [{arriveTangent=0.4778, interpMode=RCIM_Cubic, leaveTangent=0.4778, tangentMode=RCTM_User, time=0, value=0.8}, <at 0.5 1>]}
       YCurve {keys [{arriveTangent=0.4778, interpMode=RCIM_Cubic, leaveTangent=0.4778, tangentMode=RCTM_User, time=0, value=0.8}, <at 0.5 1>]}
       ZCurve {keys [{arriveTangent=0.4778, interpMode=RCIM_Cubic, leaveTangent=0.4778, tangentMode=RCTM_User, time=0, value=0.8}, <at 0.5 1>]}
      }}))
   &FN_Light_Attributes(
    Color=<rgba 0 0.35 0.1632 1>
    Create Color Attribute=true
    Create Exponent Attribute=true
    Create Radius Attribute=true
    Exponent=2
    Radius=300
    Scale Color=&FloatFromCurve(
     FloatCurve=DataInterface() {propertyValues {Curve {
        defaultValue=1
        keys [
         {arriveTangent=9.0725, interpMode=RCIM_Cubic, leaveTangent=9.0725, tangentMode=RCTM_User, time=0, value=0}
         {arriveTangent=-2.5249, interpMode=RCIM_Cubic, leaveTangent=-2.5249, tangentMode=RCTM_User, time=0.1408, value=0.9847}
         {arriveTangent=-0.0405, interpMode=RCIM_Cubic, leaveTangent=-0.0405, tangentMode=RCTM_User, time=1, value=0}
        ]
       }}}
     Scale Curve=20
    )
   )
  ]
  rendererClasses [
   &NiagaraMeshRendererProperties
   &NiagaraSpriteRendererProperties
   &NiagaraLightRendererProperties
  ]
  renderers [
   {rendererClass=&NiagaraMeshRendererProperties, rendererIndex=0}
   {rendererClass=&NiagaraSpriteRendererProperties, rendererIndex=1}
   {rendererClass=&NiagaraLightRendererProperties, rendererIndex=2}
  ]
 }
 Logo_Main {
  emitterUpdateScript [
   &EmitterState(Life Cycle Mode=::NewEnumerator1, Loop Behavior=::NewEnumerator1, Loop Delay=0.15, Loop Duration=3.3, UseLoopDelay=true) {module=EmitterState001}
   &SpawnBurst_Instantaneous(
    Spawn Count=1
    Spawn Probability=&FloatFromCurve(FloatCurve=DataInterface() {propertyValues {Curve {keys [<at 0.5 1 RCIM_Cubic>,<at 1 0>]}}}, CurveIndex=$System.NormalizedLoopAge)
    Use Spawn Probability=false
   )
  ]
  particleSpawnScript [
   &InitializeParticle(
    Color=&MaskLinearColorBySpawnGroup(Masked Linear Color=<rgba 0.2346 1 0.9473 2>, Passthrough Linear Color=<rgba 0.0235 0.1 0.0947 1>)
    Lifetime=3
    Mesh Scale=<1.7 1.3 1>
    Mesh Scale Mode=::NewEnumerator4
    Sprite Rotation Mode=::NewEnumerator2
    Sprite Size Mode=::NewEnumerator3
    Uniform Sprite Size=500
    UsePositionOffset=false
    Write Material Random=true
   ) {module=InitializeParticle001}
   &InitialMeshOrientation(Mesh Orientation Mode=::NewEnumerator1, Rotation Coordinate Space=::NewEnumerator0, Use Rotation Vector=true)
   &SpriteFacingAndAlignment(Sprite Facing=<0 0 1>, WriteAlignment=true, WriteFacing=true)
   /SetVariables(Particles.Linear Color=<rgba 0 0.1 0.0748 1>) {isSetParameters=true}
   /SetVariables(Particles.Vector:Vector3f=&MakeVectorFromLinearColorRGB(Color=<rgba 1 1 1 1>)) {isSetParameters=true}
   /SetVariables(Particles.VisibilityTag=&SpawnGroup()) {enabled=false, isSetParameters=true}
   /SetVariables(
    Particles.FlareSize:Vector2f=&Vector2DFromFloat(Value=&UniformRangedFloat(Minimum=250, Maximum=300))
    Particles.LensFlare_Color:LinearColor=<rgba 0 0.1 0.0748 0.5>
   ) {isSetParameters=true}
   &CameraOffset(Camera Offset Amount=2)
  ]
  particleUpdateScript [
   &ParticleState()
   &UpdateMeshOrientation(
    Facing Direction=<0 -1 0>
    Facing Interpolation=true
    Orientation Method=::NewEnumerator1
    Override Quat=false
    TransformFacing=false
    TransformUp=false
   )
   &ScaleMeshSize(Scale Factor=<1.25 1.25 1.25>)
   &ScaleColor(
    Linear Color Curve=DataInterface() {propertyValues {
      AlphaCurve {keys [<at 0 0 RCIM_Cubic>,<at 0.075 2>,<at 0.4998 2>,<at 1 0>]}
      BlueCurve {keys [<at 0.1696 15>,<at 1 0>]}
      GreenCurve {keys [<at 0.1696 15>,<at 1 0>]}
      RedCurve {keys [<at 0.1696 15>,<at 1 0>]}
     }}
    Scale Mode=::NewEnumerator2
   )
   &DynamicMaterialParameters(
    Index 0 Param 1=&MaskFloatBySpawnGroup(
     Masked Float=&FloatFromCurve(FloatCurve=DataInterface() {propertyValues {Curve {keys [<at 0.8 -2 RCIM_Cubic>, {arriveTangent=-1.783, tangentMode=RCTM_User, time=1, value=-3.1}]}}})
     Passthrough Float=1
    )
    Index 0 Param 2=&MaskFloatBySpawnGroup(
     Masked Float=&FloatFromCurve(FloatCurve=DataInterface() {propertyValues {Curve {keys [
          <at 0 0 RCIM_Cubic>
          {arriveTangent=2.5, leaveTangent=2.5, time=0.15, value=2}
          {arriveTangent=-2.3529, interpMode=RCIM_Cubic, leaveTangent=-2.3529, time=0.9, value=2}
          <at 1 0>
         ]}}})
     Passthrough Float=0
    )
    Index 0 Param 3=&MaskFloatBySpawnGroup(
     Masked Float=&RandomRangeFloat(Minimum=1.5, Maximum=2.55, Evaluation Type=::NewEnumerator1)
     Passthrough Float=0
    )
    Index 1 Param 1=&FloatFromCurve(FloatCurve=DataInterface() {propertyValues {Curve {
        defaultValue=1
        keys [
         {arriveTangent=124.1284, interpMode=RCIM_Cubic, time=0, value=0.5}
         {arriveTangent=2.9607, leaveTangent=2.9607, time=0.1, value=1}
         {arriveTangent=-20.0086, interpMode=RCIM_Cubic, leaveTangent=-20.0086, tangentMode=RCTM_User, time=0.85, value=1}
         {arriveTangent=-1.0087, interpMode=RCIM_Cubic, leaveTangent=-1.0087, tangentMode=RCTM_User, time=1, value=0}
        ]
       }}})
    Index 1 Param 2=2
    Module.Index0Float0Write=true
    Module.Index0Float1Write=true
    Module.Index0Float2Write=true
    Module.Index0Float3Write=true
    Module.Index1Float0Write=true
    Module.Index1Float1Write=true
    Module.Index1Float2Write=true
    Module.Index1Float3Write=true
    Module.Index2Float0Write=true
    Module.Index2Float1Write=true
    Module.Index2Float2Write=true
    Module.Index2Float3Write=true
    Write Parameter Index 1=::NewEnumerator1
    Write Parameter Index 2=::NewEnumerator1
   ) {module=DynamicMaterialParameters001}
   /SetVariables(Particles.Color=&MaskLinearColor(
     Mask Source=1
     Mask Parameter=1
     Mask=1
     Masked Linear Color=&ColorFromCurve(ColorCurve=DataInterface() {propertyValues {
        AlphaCurve {keys [<at 0.8 0.2 RCIM_Cubic>, {arriveTangent=-9.2891, tangentMode=RCTM_User, time=1, value=0}]}
        BlueCurve {keys [<at 0.5241 0.0953 RCIM_Cubic>]}
        GreenCurve {keys [<at 0.5241 0.0999 RCIM_Cubic>]}
        RedCurve {keys [<at 0.5241 0.0232 RCIM_Cubic>]}
       }})
     Passthrough Linear Color=<rgba 1 1 1 1>
    )) {enabled=false, isSetParameters=true}
   &FN_Light_Attributes(
    Color=<rgba 0 0.35 0.1632 1>
    Create Color Attribute=true
    Create Exponent Attribute=true
    Create Radius Attribute=true
    Exponent=2
    Radius=300
    Scale Color=&FloatFromCurve(
     FloatCurve=DataInterface() {propertyValues {Curve {
        defaultValue=1
        keys [
         {arriveTangent=9.0725, interpMode=RCIM_Cubic, leaveTangent=9.0725, tangentMode=RCTM_User, time=0, value=0}
         {arriveTangent=-2.5249, leaveTangent=-2.5249, time=0.1408, value=0.9847}
         {arriveTangent=-1.146, leaveTangent=-1.146, time=0.8307, value=1}
         {arriveTangent=-0.0405, interpMode=RCIM_Cubic, leaveTangent=-0.0405, tangentMode=RCTM_User, time=1, value=0}
        ]
       }}}
     Scale Curve=20
    )
   )
  ]
  rendererClasses [
   &NiagaraMeshRendererProperties
   &NiagaraLightRendererProperties
   &NiagaraSpriteRendererProperties
  ]
  renderers [
   {rendererClass=&NiagaraMeshRendererProperties, rendererIndex=0}
   {rendererClass=&NiagaraMeshRendererProperties, rendererIndex=1}
   {rendererClass=&NiagaraLightRendererProperties, rendererIndex=2}
   {rendererClass=&NiagaraSpriteRendererProperties, rendererIndex=3}
  ]
 }
}
summary {
 systemName=NS_SatireCaneOde_LightningBolt
 emitters [
  {
   emitterName=Lightning_Bolts
   rendererClasses [
    &NiagaraMeshRendererProperties
    &NiagaraSpriteRendererProperties
    &NiagaraLightRendererProperties
   ]
  }
  {
   emitterName=Lightning_Bolts_Impact
   rendererClasses [
    &NiagaraMeshRendererProperties
    &NiagaraSpriteRendererProperties
    &NiagaraLightRendererProperties
   ]
  }
  {
   emitterName=Logo_Main
   rendererClasses [
    &NiagaraMeshRendererProperties
    &NiagaraLightRendererProperties
    &NiagaraSpriteRendererProperties
   ]
  }
  {emitterName=Intro_Logo_Main001, rendererClasses [&NiagaraMeshRendererProperties]}
 ]
 userVariables [
  User.bIsFrontEnd {type=NiagaraBool}
  User.bIsFrontEndPreview {type=NiagaraBool}
  User.DecalEnabled {type=NiagaraBool, defaultValue=true}
  User.FX_Scale_Global {type=NiagaraFloat, defaultValue=0, description=""}
 ]
}
