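# Niagara effect for the SatireCane Ode emote, playing the pickaxe's glow outro as a brief flash with sparks.
# Mesh emitter spawns a slightly upscaled mesh whose color and alpha spike and fade on a curve driven by system age.
# Directional_sparks fires a slightly delayed burst of short-lived cyan sprites, stretched by speed, slowed by drag and pulled down by gravity.
# A mesh-based Lightning_Arcs emitter is present but disabled.
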NiagaraSystem /BRCosmetics/Animation/Game/MainPlayer/Emotes/SatireCane_Ode/FX/NS_SatireCaneOde_Pickaxe_Glow_Outro
dependencies {
 usedDataInterfaces [&NiagaraDataInterfaceCurve, &NiagaraDataInterfaceColorCurve, &NiagaraDataInterfaceVector2DCurve]
 usedDynamicInputs [&SimulationPosition]
 usedModules [
  &SystemState
  &self:SystemSpawnScript.SystemScriptSource.SystemScriptGraph.NiagaraNodeAssignment_0.SetVariables_#1
 ]
}
emitters {
 Directional_sparks {
  emitterUpdateScript [
   &EmitterState(Life Cycle Mode=::NewEnumerator1, Loop Behavior=::NewEnumerator1, Loop Duration=4, UseLoopDelay=false)
   &SpawnBurst_Instantaneous(
    Spawn Count=&RandomRangeInt(Minimum=8, Maximum=12)
    Spawn Time=0.15
    Use Spawn Probability=false
    UseLoopCountLimit=false
   )
  ]
  particleSpawnScript [
   &InitializeParticle(
    Color=<rgba 0.2346 1 0.9473 1>
    Lifetime Max=0.25
    Lifetime Min=0.12
    Lifetime Mode=::NewEnumerator1
    Position Offset=&RandomRangeVector(Minimum=<-10 -10 -25>, Maximum=<10 10 25>)
    Position Offset Coordinate Space=::Local
    Sprite Size Max=<2 3>
    Sprite Size Min=<1 2>
    Sprite Size Mode=::NewEnumerator2
    UsePositionOffset=true
    Write Material Random=true
   ) {module=InitializeParticle001}
   &ShapeLocation(
    Cylinder Height=20
    Cylinder Radius=2
    MaskSpawnGroup=false
    Override Seed=false
    Shape Primitive=::NewEnumerator1
    Surface Only=false
    UseOwnerScale=true
   ) {enabled=false, module=ShapeLocation001}
   &AddVelocity(Velocity=&RandomRangeVector(Minimum=<-10 -10 -35>, Maximum=<10 10 35>)) {module=AddVelocity001}
   &PointForce(Module.Use Falloff=false, Module.UseLinearForce=false)
   &ApplyInitialForces()
  ]
  particleUpdateScript [
   &ParticleState()
   &Drag(Drag=&RandomRangeFloat(Minimum=2, Maximum=3), Use Linear Drag=true, Use Rotational Drag=false)
   &GravityForce(Gravity=&RandomRangeVector(Minimum=<0 0 -980>, Maximum=<0 0 -500>))
   &SolveForcesAndVelocity(Clamp Velocity=false, Limit Acceleration=false)
   &ScaleSpriteSizeBySpeed(Max Scale Factor=<1 10>, Min Scale Factor=<0.65 2>)
   &ScaleSpriteSize(Scale Factor=&Vector2DFromCurve(Vector2Curve=DataInterface() {propertyValues {XCurve {keys [<at 0 1>,<at 1 1>]}, YCurve {keys [<at 0 1>,<at 1 0>]}}}, ScaleCurve=<1 1>))
   &ScaleColor(Scale RGB=&Multiply_VectorByFloat(Vector=<1 1 1>, Float=10), ScaleA=true, ScaleRGB=true)
  ]
  rendererClasses [&NiagaraSpriteRendererProperties]
  renderers [{rendererClass=&NiagaraSpriteRendererProperties}]
 }
 Lightning_Arcs {
  bEnabled=false
  emitterUpdateScript [
   &EmitterState(Life Cycle Mode=::NewEnumerator1, Loop Behavior=::NewEnumerator1, Loop Delay=2, Loop Duration=1, UseLoopDelay=false)
   &SpawnRate(SpawnRate=10, Use Spawn Probability=false)
   &SpawnBurst_Instantaneous(
    Spawn Count=&UniformRangedInt(Minimum=2, Maximum=3)
    Spawn Probability=&FloatFromCurve(FloatCurve=DataInterface() {propertyValues {Curve {keys [<at 0 1>,<at 1 0>]}}}, CurveIndex=$System.NormalizedLoopAge)
    Use Spawn Probability=false
   )
  ]
  particleSpawnScript [
   &InitializeParticle(
    Color=<rgba 0.2346 1 0.9473 1>
    Lifetime=&UniformRangedFloat(Minimum=0.125, Maximum=0.175)
    Mesh Scale Max=<2 3 3>
    Mesh Scale Min=<1.5 1 1>
    Mesh Scale Mode=::NewEnumerator2
    Position Offset=<0 0 -20>
    UsePositionOffset=true
    Write Material Random=true
   ) {module=InitializeParticle001}
   &DynamicMaterialParameters(
    Index 0 Param 1=6
    Index 0 Param 3=&UniformRangedFloat(Minimum=1.45, Maximum=2.2)
    Index 0 Param 4=&UniformRangedFloat(Minimum=-60, Maximum=60)
    Module.Index0Float0Write=true
    Module.Index0Float1Write=true
    Module.Index0Float2Write=true
    Module.Index0Float3Write=true
   )
   /SetVariables(
    Particles.FlareOffset:Vector3f=&UniformRangedVector(Minimum=<0 0 -0.5>, Maximum=<8 0 0.5>)
    Particles.FlareSize:Vector2f=&Vector2DFromFloat(Value=&UniformRangedFloat(Minimum=10, Maximum=15))
    Particles.LensFlare_Color:LinearColor=&MultiplyLinearColor(
     Linear Color=$Particles.Color
     Scale Factor=&MakeVector4_FromLinearColor(LinearColor=$System.RandomFlareColor)
    )
   ) {isSetParameters=true}
   &InitialMeshOrientation(
    Mesh Axis To Orient=<0 0 0>
    Mesh Orientation Mode=::NewEnumerator2
    Orientation Vector=<0 0 1>
    Rotation=&RandomRangeVector(Minimum=<0 -0.1 0>, Maximum=<1 0.1 0>)
    Use Rotation Vector=true
   ) {module=InitialMeshOrientation001}
  ]
  particleUpdateScript [
   &ParticleState()
   &DynamicMaterialParameters(
    Index 0 Param 2=&FloatFromCurve(FloatCurve=DataInterface() {propertyValues {Curve {keys [{arriveTangent=-2.9283, interpMode=RCIM_Cubic, leaveTangent=-2.9283, tangentMode=RCTM_User, time=0.1, value=1}, <at 1 -1>]}}})
    Module.Param0WriteEnabled=false
    Module.Param1WriteEnabled=true
    Module.Param2WriteEnabled=false
    Module.Param3WriteEnabled=false
   ) {module=DynamicMaterialParameters001}
   &ScaleColor(
    Scale RGB=&MakeVectorFromLinearColorRGB(Color=&ColorFromCurve(ColorCurve=DataInterface() {propertyValues {
        AlphaCurve {keys []}
        BlueCurve {keys [<at 0 2 RCIM_Cubic>,<at 1 4 RCIM_Cubic>]}
        GreenCurve {keys [<at 0 1.8653 RCIM_Cubic>,<at 1 2.6922 RCIM_Cubic>]}
        RedCurve {keys [<at 0 1.5312 RCIM_Cubic>,<at 1 0.8958 RCIM_Cubic>]}
       }}))
    ScaleA=true
    ScaleRGB=true
   )
  ]
  rendererClasses [&NiagaraMeshRendererProperties]
  renderers [{rendererClass=&NiagaraMeshRendererProperties, rendererIndex=0}]
 }
 Mesh {
  emitterUpdateScript [
   &EmitterState(Inactive Response=::NewEnumerator1, Life Cycle Mode=::NewEnumerator1, Loop Behavior=::NewEnumerator1, Loop Duration=0.7, UseLoopDelay=false)
   &SpawnBurst_Instantaneous(Use Spawn Probability=false, UseLoopCountLimit=false)
  ]
  particleSpawnScript [
   &InitializeParticle(
    Color=<rgba 1 1 1 50>
    Lifetime=0.1
    Mesh Scale Mode=::NewEnumerator3
    Mesh Uniform Scale=1.05
    Position Offset=<0 0 -0.02>
    UsePositionOffset=true
    Write Material Random=true
   )
   &InitialMeshOrientation(Mesh Orientation Mode=::NewEnumerator7, Rotation=<0 0 0.25>, Use Rotation Vector=true)
  ]
  particleUpdateScript [&ScaleColor(
    Curve Index=$System.Age
    Linear Color Curve=DataInterface() {propertyValues {
      AlphaCurve {keys [
        <at 0 0 RCIM_Cubic>
        {arriveTangent=-4.0465, interpMode=RCIM_Cubic, leaveTangent=-4.0465, tangentMode=RCTM_User, time=0.2, value=1}
        <at 0.9139 0>
       ]}
      BlueCurve {keys [
        <at 0 0 RCIM_Cubic>
        {arriveTangent=-8.1681, interpMode=RCIM_Cubic, leaveTangent=-8.1681, tangentMode=RCTM_User, time=0.2, value=2}
        <at 1 0>
       ]}
      GreenCurve {keys [
        <at 0 0 RCIM_Cubic>
        {arriveTangent=-8.1681, interpMode=RCIM_Cubic, leaveTangent=-8.1681, tangentMode=RCTM_User, time=0.2, value=2}
        <at 1 0>
       ]}
      RedCurve {keys [
        <at 0 0 RCIM_Cubic>
        {arriveTangent=-8.1681, interpMode=RCIM_Cubic, leaveTangent=-8.1681, tangentMode=RCTM_User, time=0.2, value=2}
        <at 1 0>
       ]}
     }}
    Scale Mode=::NewEnumerator2
   )]
  rendererClasses [&NiagaraMeshRendererProperties]
  renderers [{rendererClass=&NiagaraMeshRendererProperties, rendererIndex=0}]
 }
}
summary {
 systemName=NS_SatireCaneOde_Pickaxe_Glow_Outro
 emitters [
  {bEnabled=false, emitterName=Lightning_Arcs, rendererClasses [&NiagaraMeshRendererProperties]}
  {emitterName=Mesh, rendererClasses [&NiagaraMeshRendererProperties]}
  {emitterName=Directional_sparks, rendererClasses [&NiagaraSpriteRendererProperties]}
 ]
 userVariables [
  User.bIsFrontEnd {type=NiagaraBool}
  User.bIsFrontEndPreview {type=NiagaraBool}
  User.DecalEnabled {type=NiagaraBool, defaultValue=true}
  User.FX_Scale_Global {type=NiagaraFloat, defaultValue=0, description=""}
  User.SkeletalMesh {
   type=NiagaraDataInterfaceSkeletalMesh
   defaultValue {
    struct=NiagaraExt_VariableValue_DataInterface
    value {dataInterface=&self:NiagaraDataInterfaceSkeletalMesh_0}
   }
  }
 ]
}
