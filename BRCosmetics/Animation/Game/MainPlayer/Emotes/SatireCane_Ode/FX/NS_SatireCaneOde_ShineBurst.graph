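# Niagara system for a very brief shine burst in the SatireCane Ode emote.
# A single MainShape emitter bursts short-lived particles, each drawn by several sprite renderers and a mesh renderer for a layered flare.
# Curves fade alpha and grow size, while sprites and mesh rotate.
# Derived attributes give corona, dark glow and central flare layers their own sizes or colors, including a rainbow-cycling flare, and randomized dynamic parameters feed the materials.
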NiagaraSystem /BRCosmetics/Animation/Game/MainPlayer/Emotes/SatireCane_Ode/FX/NS_SatireCaneOde_ShineBurst
dependencies {
 usedDataInterfaces [&NiagaraDataInterfaceCurve, &NiagaraDataInterfaceColorCurve, &NiagaraDataInterfaceSkeletalMesh]
 usedDynamicInputs []
 usedModules [
  &self:SystemSpawnScript.SystemScriptSource.SystemScriptGraph.NiagaraNodeAssignment_0.SetVariables_#1
  &SystemState
 ]
}
emitters {MainShape {
  emitterUpdateScript [
   &EmitterState()
   &SpawnBurst_Instantaneous(Use Spawn Probability=false, UseLoopCountLimit=false)
  ]
  particleSpawnScript [
   &InitializeParticle(
    Lifetime=0.18
    Mass Mode=::NewEnumerator1
    Mesh Scale Mode=::NewEnumerator3
    Mesh Uniform Scale=&FloatFromCurve(
     FloatCurve=DataInterface() {propertyValues {Curve {keys [{tangentMode=RCTM_User, time=0, value=1}, {tangentMode=RCTM_User, time=1, value=2}]}}}
     CurveIndex=&ReturnNormalizedExecIndex()
     Scale Curve=2
    )
    Position=&SimulationPosition()
    Position Mode=::NewEnumerator1
    Sprite Size Mode=::NewEnumerator3
    Uniform Sprite Size=40
    UsePositionOffset=true
    Write Material Random=true
   ) {module=InitializeParticle001}
   &CameraOffset(Camera Offset Amount=5)
  ]
  particleUpdateScript [
   &ParticleState()
   &SpriteRotationRate(Rotation Rate=200)
   &UpdateMeshOrientation(Rotation Rate=&FloatFromCurve(FloatCurve=DataInterface() {propertyValues {Curve {defaultValue=1, keys [<at 0 0 RCIM_Cubic>, {tangentMode=RCTM_User, time=0.8, value=0.5}]}}}))
   &ScaleColor(
    Scale Alpha=&FloatFromCurve(FloatCurve=DataInterface() {propertyValues {Curve {keys [
         {arriveTangent=6.4019, interpMode=RCIM_Cubic, leaveTangent=6.4019, tangentMode=RCTM_User, time=0, value=0}
         {arriveTangent=0.002, interpMode=RCIM_Cubic, leaveTangent=0.002, time=0.1, value=0.7}
         {arriveTangent=-0.9006, interpMode=RCIM_Cubic, leaveTangent=-0.9006, tangentMode=RCTM_User, time=1.0011, value=0.002}
        ]}}})
    ScaleA=true
    ScaleRGB=true
   )
   &ScaleSpriteSize(Scale Factor=&Multiply_Vector2D_ByFloat(
     Float=&FloatFromCurve(FloatCurve=DataInterface() {propertyValues {Curve {keys [
          {arriveTangent=1.9662, interpMode=RCIM_Cubic, leaveTangent=1.9662, tangentMode=RCTM_User, time=0, value=0.4}
          {interpMode=RCIM_Cubic, tangentMode=RCTM_User, time=1, value=1}
         ]}}})
     Vector2D=<1 1>
    ))
   &ScaleMeshSize(Scale Factor=&VectorFromFloat(Value=&FloatFromCurve(
      FloatCurve=DataInterface() {propertyValues {Curve {defaultValue=0, keys [<at 0 0 RCIM_Cubic>, {tangentMode=RCTM_User, time=1, value=0.5}]}}}
      Scale Curve=0.35
     )))
   /SetVariables(
    Particles.DynamicParameter_01:Vector4f=&MakeVector4(
     X=&RandomRangeFloat(Minimum=0, Maximum=1)
     Y=2.5
     Z=&RandomRangeFloat(Minimum=2, Maximum=3)
     W=0
    )
    Particles.DynamicParameter_02:Vector4f=&MakeVector4(X=0, Y=&RandomRangeFloat(Minimum=0, Maximum=1), Z=0, W=0)
   ) {isSetParameters=true}
   /SetVariables(Particles.Corona_Size:Vector2f=&Multiply_Vector2D_ByFloat(
     Float=&FloatFromCurve(FloatCurve=DataInterface() {propertyValues {Curve {keys [<at 0 0.1 RCIM_Cubic>, {arriveTangent=1, time=0.2, value=0.3}, <at 0.8 1>]}}})
     Vector2D=$Particles.SpriteSize
    )) {isSetParameters=true}
   /SetVariables(Particles.DarkGlowSize:Vector2f=&Multiply_Vector2D_ByFloat(Float=0.7, Vector2D=$Particles.SpriteSize)) {isSetParameters=true}
   /SetVariables(Particles.DargGlowColor:LinearColor=&ColorFromCurve(ColorCurve=DataInterface() {propertyValues {
       AlphaCurve {keys [<at 0 0.7 RCIM_Cubic>,<at 1 0 RCIM_Cubic>]}
       BlueCurve {keys [<at 0 0.02 RCIM_Cubic>]}
       GreenCurve {keys [<at 0 0.0133 RCIM_Cubic>]}
       RedCurve {keys [<at 0 0.0158 RCIM_Cubic>]}
      }})) {isSetParameters=true}
   /SetVariables(Particles.CentralFlareSize:Vector2f=&Multiply_Vector2D_ByFloat(Float=3, Vector2D=$Particles.SpriteSize)) {isSetParameters=true}
   /SetVariables(Particles.CentralFlareColor:LinearColor=&MultiplyLinearColor(
     Linear Color=&ColorFromCurve(
      ColorCurve=DataInterface() {propertyValues {
        AlphaCurve {keys [<at 0 1>,<at 1 0>]}
        BlueCurve {keys [<at 0 0.2>,<at 0.333 0.2>,<at 0.667 1>,<at 1 0.2>]}
        GreenCurve {keys [<at 0 0.2>,<at 0.333 1>,<at 0.667 0.2>,<at 1 0.2>]}
        RedCurve {keys [<at 0 1>,<at 0.333 0.2>,<at 0.667 0.2>,<at 1 1>]}
       }}
      CurveIndex=&Multiply_Float(A=$Particles.NormalizedAge, B=2)
     )
     Scale Factor=&MakeFloat4FromVectorAndFloat("Vector (RGB)"=&VectorFromFloat(Value=2), "Float (Alpha)"=0.5)
    )) {isSetParameters=true}
  ]
  rendererClasses [&NiagaraSpriteRendererProperties, &NiagaraMeshRendererProperties]
  renderers [
   {rendererClass=&NiagaraSpriteRendererProperties}
   {rendererClass=&NiagaraSpriteRendererProperties, rendererIndex=1}
   {rendererClass=&NiagaraSpriteRendererProperties, rendererIndex=2}
   {rendererClass=&NiagaraSpriteRendererProperties, rendererIndex=3}
   {rendererClass=&NiagaraMeshRendererProperties, rendererIndex=4}
   {rendererClass=&NiagaraSpriteRendererProperties, rendererIndex=5}
  ]
 }}
summary {
 systemName=NS_SatireCaneOde_ShineBurst
 emitters [{emitterName=MainShape, rendererClasses [&NiagaraSpriteRendererProperties, &NiagaraMeshRendererProperties]}]
 userVariables [
  User.bIsFrontEnd {type=NiagaraBool}
  User.bIsFrontEndPreview {type=NiagaraBool}
  User.BoneIndex {type=NiagaraFloat, defaultValue=0, description=""}
 ]
}
